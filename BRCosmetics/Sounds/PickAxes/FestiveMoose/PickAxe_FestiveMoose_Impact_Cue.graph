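# Impact sound cue for the Festive Moose cosmetic pickaxe, pairing a random impact hit with an occasional moose call.
# A mixer layers one of six equally weighted impacts with a quieter call branch, delayed slightly further behind the cue's short lead-in delay.
# Nested random nodes with an empty slot and weighted silence waves make calls play on roughly one hit in ten.
# Pickaxe impact attenuation and melee concurrency settings are assigned.
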SoundCue /BRCosmetics/Sounds/PickAxes/FestiveMoose/PickAxe_FestiveMoose_Impact_Cue
root=Delay_1
nodes [
 {id=Delay_0, children [Random_1], properties {delayMax=0.1, delayMin=0.1}}
 {id=Delay_1, children [Modulator_0], properties {delayMax=0.1, delayMin=0.1}}
 {id=Mixer_0, children [Delay_0,Random_2], properties {inputVolume [0.6,1]}}
 {id=Modulator_0, children [Mixer_0], properties {volumeMax=1, volumeMin=1}}
 {
  id=Random_0
  children [WavePlayer_10, WavePlayer_11, WavePlayer_12, WavePlayer_13, WavePlayer_13, WavePlayer_13]
  properties {bShouldExcludeFromBranchCulling=true, weights [0.25,0.25,0.25,1,1,1]}
 }
 {id=Random_1, children [Random_0,null], properties {bShouldExcludeFromBranchCulling=true, weights [1,1]}}
 {
  id=Random_2
  children [WavePlayer_0,WavePlayer_1,WavePlayer_2,WavePlayer_3,WavePlayer_8,WavePlayer_9]
  properties {bShouldExcludeFromBranchCulling=true, weights [1,1,1,1,1,1]}
 }
 {id=WavePlayer_0, properties {soundWaveAssetPtr=&/Pickaxe_FestiveMoose_Impact_01_v1_Mix}}
 {id=WavePlayer_1, properties {soundWaveAssetPtr=&/Pickaxe_FestiveMoose_Impact_02_v1_Mix}}
 {id=WavePlayer_10, properties {soundWaveAssetPtr=&/Pickaxe_FestiveMoose_MooseCall_01}}
 {id=WavePlayer_11, properties {soundWaveAssetPtr=&/Pickaxe_FestiveMoose_MooseCall_02}}
 {id=WavePlayer_12, properties {soundWaveAssetPtr=&/Pickaxe_FestiveMoose_MooseCall_03}}
 {id=WavePlayer_13, properties {soundWaveAssetPtr=&/1second_silence_FestiveMoose}}
 {id=WavePlayer_2, properties {soundWaveAssetPtr=&/Pickaxe_FestiveMoose_Impact_03_v1_Mix}}
 {id=WavePlayer_3, properties {soundWaveAssetPtr=&/Pickaxe_FestiveMoose_Impact_04_v1_Mix}}
 {id=WavePlayer_8, properties {soundWaveAssetPtr=&/Pickaxe_FestiveMoose_Impact_05_v1_Mix}}
 {id=WavePlayer_9, properties {soundWaveAssetPtr=&/Pickaxe_FestiveMoose_Impact_06_v1_Mix}}
]
properties {
 attenuationSettings=/Game/Sounds/Attenuation/Fort_Pickaxe_Impact_Att
 audioPropertiesSheet=/BRCosmetics/Sounds/Settings/PropertySheets/Pickaxe/APS_Cosmetic_Pickaxe_Impact
 bExcludeFromRandomNodeBranchCulling=true
 bPrimeOnLoad=true
 maxDistance=4300
 soundClassObject=&Fort_Impacts
 attenuationOverrides {
  bAttenuateWithLPF=true
  distanceAlgorithm=Logarithmic
  falloffDistance=800
  lPFRadiusMax=800
  lPFRadiusMin=400
  attenuationShapeExtents {x=300}
 }
 concurrencyOverrides {bLimitToOwner=true, maxCount=2, resolutionRule=StopOldest}
 concurrencySet [&Player_Melee_Weapon_Conc]
}
